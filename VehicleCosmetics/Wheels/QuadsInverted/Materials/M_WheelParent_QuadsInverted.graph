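# Parent material for QuadsInverted vehicle wheels, adding an animated custom-colored emissive effect to the simple wheel base function.
# Rim glow layers three radial sweeps rotating at different speeds on UV1, masked by an inverted rim texture; tire glow sharpens a panning stripe through a mask.
# Vertex blue and red channels place rim and tire effects.
# The result fully replaces base emissive, since its blend alpha sums to one, then feeds damage and de-rez functions.
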Material /VehicleCosmetics/Wheels/QuadsInverted/Materials/M_WheelParent_QuadsInverted
graph {
 properties {
  bAutomaticallySetUsageInEditor=false
  bUseMaterialAttributes=true
  bUsedWithMeshDeformer=true
  bUsedWithSkeletalMesh=true
  pixelDepthOffsetMode=PDOM_Legacy
  usedShadingModels=Default Lit
 }
 nodes [
  n1 = MaterialFunctionCall() {materialFunction=&MF_WheelParent_Simple}
  n2 = GetMaterialAttributes(None=@n1.MaterialAttributes) {attributeGetTypes [#1]}
  n3 = LinearInterpolate(A=@n2.Emissive Color, B=NamedRerouteUsage(), Alpha=Add(A=@n1.VertexColorR, B=OneMinus(None=@n1.VertexColorR)))
  MaterialFunctionCall(In=MaterialFunctionCall(Material=SetMaterialAttributes(MaterialAttributes=@n2.MaterialAttributes, Emissive Color=@n3) {attributeSetTypes [#1]}) {materialFunction=&MF_BaseVehicle_Damage}.Result) {materialFunction=&MF_VehicleDeRez_Transition}
  n4 = VectorParameter() {group=Effect, parameterName=Emissive_Custom_Color, defaultValue=<rgba 1 0.02 0 1>}
  n5 = TextureCoordinate() {coordinateIndex=1}
  n6 = Rotator(Coordinate=@n5) {speed=4}
  n7 = Rotator(Coordinate=@n5) {speed=8}
  n8 = Add(
   A=Add(
    A=TextureSample(UVs=@n6) {samplerType=SAMPLERTYPE_LinearColor, texture=/VehicleCosmetics/Wheels/_SharedTextures/FX/T_RadialSweep_M}.G
    B=TextureSample(UVs=@n7) {samplerType=SAMPLERTYPE_LinearColor, texture=&T_RadialSweep_M}.B
   )
   B=TextureSample(UVs=Rotator(Coordinate=@n5) {speed=6}) {samplerType=SAMPLERTYPE_LinearColor, texture=&T_RadialSweep_M}.R
  )
  n9 = Constant() {r=8}
  n10 = TextureSampleParameter2D() {
   group=Effect
   parameterName=Rim_M_Efect
   samplerType=SAMPLERTYPE_LinearColor
   texture=/VehicleCosmetics/Wheels/QuadsInverted/Textures/T_QuadsInverted_Rim_RGB_01
  }
  n11 = Clamp(None=Add(A=@n10.R, B=@n10.R))
  n12 = Multiply(A=Desaturation(None=@n4.RGB, Fraction=Constant() {r=0.066}), B=Multiply(A=Multiply(A=@n8, B=@n9), B=OneMinus(None=@n11)))
  n13 = VertexColor()
  n14 = OneMinus(None=@n13.B)
  n15 = TextureSampleParameter2D() {
   group=Effect
   parameterName=Tire_M
   samplerType=SAMPLERTYPE_LinearColor
   texture=/VehicleCosmetics/Wheels/QuadsInverted/Textures/T_QuadsInverted_Tire_MASK1
  }
  n16 = TextureSample(UVs=Panner() {speedX=-0.5}) {texture=/VehicleCosmetics/Wheels/_SharedTextures/FX/T_GradientStripeVertical_M}
  n17 = Clamp(None=Subtract(A=@n15.R, B=OneMinus(None=@n16.G)))
  n18 = Constant() {r=8}
  n19 = Multiply(A=Clamp(None=Power(Base=@n17, Exp=@n18)), B=Constant() {r=3})
  NamedRerouteDeclaration(None=LinearInterpolate(A=LinearInterpolate(B=@n12, Alpha=@n14), B=Multiply(A=@n19, B=@n4.RGB), Alpha=@n13.R)) {name=Emissive, nodeColor=<rgba 1 0.75 0 1>}
 ]
}
